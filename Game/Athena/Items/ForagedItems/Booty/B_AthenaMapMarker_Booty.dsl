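; Map marker actor for pirate Booty treasure, showing an X decal on the ground and a sky beacon.
; BeginPlay creates the decal's dynamic material and line-traces vertically to align the decal to the ground normal.
; The chest's passive ability calls UnhideSkyMarker and HideSkyMarker, which toggle the Niagara indicator and play one timeline forward or reversed.
; The timeline animates X, pillar and decal materials; unhiding also plays a sound, and finished hiding hides the decal.
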(event EventBeginPlay
  ;[1 1 1] Setup Decal Material Instance
  (bind _returnvalue (Rendering|Components|Decal|CreateDynamicMaterialInstance (Variables|Default|GetDecal)))
  (Variables|Default|SetMI_Decal _returnvalue)
  ;[1 1 1] FX - Nick P - Rotates decal to face surface normal of ground
  (bind (_outhit _returnvalue_2) (Collision|LineTraceByChannel (Math|Vector|MakeVector (.x (Transformation|GetWorldLocation (Variables|Default|GetDecal))) (.y (Transformation|GetWorldLocation (Variables|Default|GetDecal))) (+ (.z (Transformation|GetWorldLocation (Variables|Default|GetDecal))) (Variables|Default|GetDecalTraceRange))) (Math|Vector|MakeVector (.x (Transformation|GetWorldLocation (Variables|Default|GetDecal))) (.y (Transformation|GetWorldLocation (Variables|Default|GetDecal))) (- (.z (Transformation|GetWorldLocation (Variables|Default|GetDecal))) (Variables|Default|GetDecalTraceRange))) "TraceTypeQuery10"))
  (if _returnvalue_2
    (Transformation|SetWorldRotation (Variables|Default|GetDecal) (Utilities|Select (Math|Conversions|RotationFromXVector (.ImpactNormal (Collision|BreakHitResult _outhit))) (Math|Rotator|MakeRotator (.roll (Math|Conversions|RotationFromXVector (.ImpactNormal (Collision|BreakHitResult _outhit)))) -9e+01 (.yaw (Math|Conversions|RotationFromXVector (.ImpactNormal (Collision|BreakHitResult _outhit))))) (> (.pitch (Math|Conversions|RotationFromXVector (.ImpactNormal (Collision|BreakHitResult _outhit)))) -45.0)) false false)))

(event Custom|UnhideSkyMarker
  (bind _mid_glowy_x (Variables|Default|GetMID_Glowy_X))
  ;[1 1 1] Hide and Unhide then sky Marker. This is called from the passive ability that opperates the chest.
  (if (Variables|Default|GetShouldHide)
    (Development|SetHiddeninGame (Variables|Default|GetDecal) false false)
    (Audio|PlaySoundatLocation "/Game/Sounds/Foley_Loot/Containers/Pirate_Booty/Pirate_Booty_Beacon_Activate_Cue.Pirate_Booty_Beacon_Activate_Cue" (Transformation|GetActorLocation))
    (Variables|Default|SetShouldHide false)
    (Niagara|SetNiagaraVariable(Bool) (Variables|Default|GetNS_PirateChest_Indicator) "IsVisible" true)
    (bind (_direction _opacity _spawnscale _spawnverticalscale _scale_x_mark_on_z _scale_x_mark_on_xy _scalecorona _pillarscale _secondaryopacity _opacity_x) (Timeline|SpawnMarkerAnim PlayFromStart)
      (:Update
        (Rendering|Material|SetScalarParameterValue _mid_glowy_x "Scale_Z" _scale_x_mark_on_z :!class /Script/Engine.MaterialInstanceDynamic)
        (Rendering|Material|SetScalarParameterValue _mid_glowy_x "Scale_XY" _scale_x_mark_on_xy :!class /Script/Engine.MaterialInstanceDynamic)
        (Rendering|Material|SetScalarParameterValue (Variables|Default|GetMID_Glow_Pillar_Mesh) "MaxScaleIncrease" _pillarscale :!class /Script/Engine.MaterialInstanceDynamic)
        (Rendering|Material|SetScalarParameterValue (Variables|Default|GetMI_Decal) "X Decal Fire Lerp" _opacity_x :!class /Script/Engine.MaterialInstanceDynamic))
      (:Finished
        (if (Variables|Default|GetShouldHide)
          (Development|SetHiddeninGame (Variables|Default|GetDecal) true false))))))

(event Custom|HideSkyMarker
  (bind _mid_glowy_x (Variables|Default|GetMID_Glowy_X))
  ;[1 1 1] Hide and Unhide then sky Marker. This is called from the passive ability that opperates the chest.
  (Variables|Default|SetShouldHide true)
  (Niagara|SetNiagaraVariable(Bool) (Variables|Default|GetNS_PirateChest_Indicator) "IsVisible" false)
  (Timeline|SpawnMarkerAnim ReverseFromEnd))

(fn ConstructionScript)
