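; Weapon Blueprint for the Athena Booty foraged item, signaling equip and unequip to gameplay abilities on the holder.
; On equip and attach, it waits briefly so the ability can finish setup, then sends the Booty Equipped gameplay event to the instigator and plays the pirate booty equip sound there.
; On local unequip, it immediately sends the Unequipped event and plays the matching sound.
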(event FortWeapon|EventOnEquippedAndAttached
  (bind _returnvalue (Game|GetInstigator))
  (Utilities|IsValid _returnvalue
    (:Is%20Valid
      (Utilities|FlowControl|Delay 0.2) ; Slight delay to ensure the GA has\nenough time to setup.
      (Ability|SendGameplayEventtoActor _returnvalue "(TagName=\"Event.Athena.Booty.Equipped\")")
      (Audio|PlaySoundatLocation "/Game/Sounds/Foley_Loot/Containers/Pirate_Booty/Pirate_Booty_Equip_Cue.Pirate_Booty_Equip_Cue" (Transformation|GetActorLocation _returnvalue)))))

(event FortWeapon|EventOnLocalUnEquip
  (bind _returnvalue (Game|GetInstigator))
  (Utilities|IsValid _returnvalue
    (:Is%20Valid
      (Ability|SendGameplayEventtoActor _returnvalue "(TagName=\"Event.Athena.Booty.Unequipped\")")
      (Audio|PlaySoundatLocation "/Game/Sounds/Foley_Loot/Containers/Pirate_Booty/Pirate_Booty_UnEquip_Cue.Pirate_Booty_UnEquip_Cue" (Transformation|GetActorLocation _returnvalue)))))

(fn ConstructionScript
  (|CallParentFunction))
