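; Booty pickup spawner that links a foraged treasure map pickup to its already existing Booty chest prop.
; On BeginPlay the server binds the pickup's OnPickup and, after a short delay, the chest's interact event.
; Pickup sends the picker a PassRef gameplay event carrying the prop reference, then destroys the spawner.
; A searched chest triggers map-rip particles and sound at the pickup, then destroys both pickup and spawner.
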(event EventBeginPlay
  (bind _pickuptobind (Variables|Default|GetPickupToBind))
  ;[1 1 1] Bind Pickup on begin play
  (if (Networking|IsServer)
    (Utilities|IsValid _pickuptobind
      (:Is%20Valid
        (Interaction|BindEventtoOnPickup _pickuptobind (delegate Custom|PassMapOnPickup))
        (Utilities|FlowControl|Delay 0.1)
        ;[1 1 1] Bind to interact complete which is the search
        (Interaction|BindEventtoOnInteract (Variables|/Game/Athena/Items/ForagedItems/Booty/Prop_Athena_Booty.Prop_Athena_Booty_C|GetChestObject (Variables|Default|GetPropRef)) (delegate Custom|ChestOpened))))))

(event Custom|Play%20Client%20FX
  (params (Loc :type Vector) (Rot_Roll :type Float%20%28single-precision%29) (Rot_Pitch :type Float%20%28single-precision%29) (Rot_Yaw :type Float%20%28single-precision%29))
  ;[1 1 1] Bind to interact complete which is the search
  (Effects|SpawnEmitteratLocation (Variables|Default|GetMapRip) Loc (Math|Rotator|MakeRotator -45.0 -3e+01 Rot_Yaw) "1.500000,1.500000,1.500000" true "None" true)
  (Audio|PlaySoundatLocation (Variables|Default|GetRipSound) Loc)
  (Development|PrintString "FX played"))

(event Custom|ChestOpened
  (params (SelfActor :type Building%20Actor%20Object%20Reference) (InteractingPawn :type Fort%20Pawn%20Object%20Reference))
  ;[1 1 1] Bind to interact complete which is the search
  (Actor|SetOwner InteractingPawn)
  (Utilities|FlowControl|Delay 0.2)
  (if (Variables|/Script/FortniteGame.BuildingContainer|GetbAlreadySearched (Variables|/Game/Athena/Items/ForagedItems/Booty/Prop_Athena_Booty.Prop_Athena_Booty_C|GetChestObject (Variables|Default|GetPropRef)))
    (CallFunction|PlayClientFX (Transformation|GetActorLocation (Variables|Default|GetPickupToBind)) (Transformation|GetActorRotation (Variables|Default|GetPickupToBind)))
    (Actor|DestroyActor (Variables|Default|GetPickupToBind))
    (Utilities|FlowControl|Delay 0.2) ; Give a moment for FX to spawn then destroy self
    (Actor|DestroyActor)))

(event Custom|PassMapOnPickup
  (params (SelfActor :type Fort%20Pickup%20Object%20Reference) (InteractingPawn :type Fort%20Pawn%20Object%20Reference))
  (bind _propref (Variables|Default|GetPropRef))
  ;[1 1 1] When picked up, send event to the person who picked it up to tell them there is already a chest.
  (if (not (GameplayTags|TagContainer|CompareTagAssetInterfacetoOtherTags (Variables|Default|GetT_BootyActive) InteractingPawn))
    (Utilities|FlowControl|Delay 0.1))
  (Development|PrintString (Utilities|String|Append "Pass Chest To: " (Utilities|GetDisplayName InteractingPawn)) true true "(R=0.208461,G=1.000000,B=0.000000,A=1.000000)")
  (Utilities|IsValid _propref
    (:Is%20Valid
      (Ability|SendGameplayEventtoActor InteractingPawn "(TagName=\"Event.Athena.Booty.PassRef\")" (Utilities|Struct|MakeGameplayEventData :OptionalObject _propref :ContextHandle "()" :InstigatorTags "(GameplayTags=)" :TargetTags "(GameplayTags=)" :EventMagnitude 0.0 :TargetData "()")))
    (:Is%20Not%20Valid
      _)) ; Make sure we have a va;id prop before sending the event.
  (Actor|DestroyActor))

(fn ConstructionScript)
