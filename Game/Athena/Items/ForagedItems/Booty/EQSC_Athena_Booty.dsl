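; EQS context Blueprint for the Athena Booty foraged item that supplies a single location to environment queries.
; ProvideSingleLocation keeps the querier actor's X and Y but replaces Z with the local variable ZLoc, which the graph never assigns.
; The location is returned only on the valid branch of an IsValid check on the querier actor.
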(fn ProvideSingleLocation
  (params (QuerierObject :type Object%20Reference) (QuerierActor :type Actor%20Object%20Reference))
  (bind (_returnvalue_x _returnvalue_y) (Math|Vector|BreakVector (Transformation|GetActorLocation QuerierActor)))
  (Utilities|IsValid QuerierActor
    (:Is%20Valid
      (return _returnvalue_x _returnvalue_y (LocalVariables|GetZLoc)))))
