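# Environment query that picks a world location for the Athena Booty foraged item.
# A spiral-pattern donut of candidate points around the EQSC_Athena_Booty context is traced down onto geometry.
# Auto-sort is off, so tests keep their authored order: projection, overlaps, water, flat surface, querier distance, randomness, water volume, building.
# Water checks reject wet points, while linear 2D distance scoring favors points farther from the querier, mixed with random scoring.
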EnvQuery /Game/Athena/Items/ForagedItems/Booty/EQS_Athena_Booty
options [{
  generator {
   name=EnvQueryGenerator_Donut_1
   class=EnvQueryGenerator_Donut
   properties {
    bAutoSortTests=false
    bUseSpiralPattern=true
    center=&/EQSC_Athena_Booty.EQSC_Athena_Booty_C
    innerRadius=2500
    numberOfRings=6
    outerRadius=15000
    pointsPerRing=8
    arcDirection {dirMode=Rotation}
    projectionData {extentX=0, navigationFilter=null, postProjectionVerticalOffset=0, traceMode=None}
   }
  }
  tests [
   EnvQueryTest_Project_4 {
    class=EnvQueryTest_Project
    properties {
     scoringFactor=1
     testComment=""
     testPurpose=Filter
     projectionData {
      navigationFilter=null
      postProjectionVerticalOffset=0
      projectDown=22000
      projectUp=0
      serializedChannel=ECC_GameTraceChannel6
      traceChannel=TraceTypeQuery7
      traceMode=GeometryByChannel
     }
    }
   }
   EnvQueryTest_Overlap_2 {
    class=EnvQueryTest_Overlap
    properties {overlapData {extentX=100, extentY=100, extentZ=100, overlapChannel=ECC_GameTraceChannel1, shapeOffset=<0 0 300>}}
   }
   EnvQueryTest_Overlap_0 {
    class=EnvQueryTest_Overlap
    properties {testPurpose=FilterAndScore, overlapData {extentX=350, extentY=350, extentZ=350, overlapChannel=ECC_GameTraceChannel15, shapeOffset=<0 0 0>}}
   }
   FortQueryTest_InsideWater_2 {
    class=FortQueryTest_InsideWater
    properties {
     bDefineReferenceValue=false
     boolValue=false
     clampMaxType=None
     clampMinType=None
     filterType=Match
     floatValueMax=0
     floatValueMin=0
     multipleContextFilterOp=AllPass
     multipleContextScoreOp=AverageScore
     normalizationType=Absolute
     referenceValue=0
     scoreClampMax=0
     scoreClampMin=0
     scoringEquation=Constant
     scoringFactor=10
     testComment=""
     testPurpose=Filter
     testRadius=512
    }
   }
   FortQueryTest_OnFlatSurface_2 {class=FortQueryTest_OnFlatSurface, properties {radius=200, scoringFactor=5, toleranceZ=50, traceOffsetUp=100}}
   EnvQueryTest_Distance_1 {
    class=EnvQueryTest_Distance
    properties {
     distanceTo=&EnvQueryContext_Querier
     filterType=Range
     floatValueMax=5000
     floatValueMin=1
     scoringEquation=Linear
     scoringFactor=5
     testComment=""
     testMode=Distance2D
     testPurpose=Score
    }
   }
   FortQueryTest_Random_1 {class=FortQueryTest_Random, properties {floatValueMax=1, scoringFactor=5, testPurpose=FilterAndScore}}
   EnvQueryTest_Volume_0 {
    class=EnvQueryTest_Volume
    properties {boolValue=false, testPurpose=Filter, volumeClass=/Script/FortniteGame.WaterVolume_Athena, volumeContext=null}
   }
   FortQueryTest_InsideBuilding_0 {class=FortQueryTest_InsideBuilding, properties {testPurpose=FilterAndScore}}
  ]
 }]
