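; GA_Athena_Booty_Input is a gameplay ability that forwards booty map input to the player's passive Booty ability so it can spawn arrows.
; On activation it runs only on the server and prints a MapInputPressed debug string.
; If the avatar actor is valid, it sends the Event.Athena.Booty.Input gameplay event to it, then ends the ability.
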(event Ability|EventActivateAbility
  (bind _returnvalue (Ability|GetAvatarActorfromActorInfo))
  ;[1 1 1] Send an event to the player's Passive Booty ability to have it spawn the arrows.
  (if (Networking|IsServer)
    (Development|PrintString "MapInputPressed")
    (Utilities|IsValid _returnvalue
      (:Is%20Valid
        (Ability|SendGameplayEventtoActor _returnvalue "(TagName=\"Event.Athena.Booty.Input\")")
        (Ability|EndAbility)))))
