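# Niagara system that shows a trail of direction indicators pointing from the effect's owner toward a booty actor position supplied through User.BootyActorPos.
# One emitter derives a normalized direction and a clamped distance alpha, which lengthens and enlarges the trail as the target gets farther away.
# Five bursts staggered 0.1 seconds apart use spawn group to pick offset, scale and mesh, with the last using an alternate mesh.
# Meshes pop in and shrink out, paired with glow sprites driven by mesh orientation and scale.
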NiagaraSystem /Game/Athena/Items/ForagedItems/Booty/NS_Athena_BootyDirection
dependencies {
 usedDataInterfaces [
  &NiagaraDataInterfaceArrayFloat
  &NiagaraDataInterfaceArrayFloat3
  &NiagaraDataInterfaceArrayInt32
  &NiagaraDataInterfaceVectorCurve
  &NiagaraDataInterfaceCurve
 ]
 usedDynamicInputs []
 usedModules [&SystemState]
}
emitters {NE_Direction_Indicators {
  emitterUpdateScript [
   &EmitterState()
   /SetVariables(Emitter.DirectionVector:Vector3f=&NormalizeVector(
     Vector To Normalize=&Subtract_Position(A=&AddVectorToPosition(Position=$User.BootyActorPos, Vector=<0 0 0>), B=$Engine.Owner.Position)
     Safe Normalize=true
     Safe Fallback Threshold=0
     Safe Fallback Vector=<0 0 0.001>
    )) {isSetParameters=true}
   /SetVariables(Emitter.AlphaFromDistance:NiagaraFloat=&ClampFloat(Float=&Divide_Float(
      A=&Subtract_Float(
       A=&VectorLength(Vector=&Subtract_Position(A=&AddVectorToPosition(Position=$User.BootyActorPos, Vector=<0 0 75>), B=$Engine.Owner.Position))
       B=100
      )
      B=250
     ))) {isSetParameters=true}
   &SpawnBurst_Instantaneous(Use Spawn Probability=false, UseLoopCountLimit=false)
   &SpawnBurst_Instantaneous(Spawn Group=1, Spawn Time=0.1, Use Spawn Probability=false, UseLoopCountLimit=false) {module=SpawnBurst_Instantaneous001}
   &SpawnBurst_Instantaneous(Spawn Group=2, Spawn Time=0.2, Use Spawn Probability=false, UseLoopCountLimit=false) {module=SpawnBurst_Instantaneous002}
   &SpawnBurst_Instantaneous(Spawn Group=3, Spawn Time=0.3, Use Spawn Probability=false, UseLoopCountLimit=false) {module=SpawnBurst_Instantaneous003}
   &SpawnBurst_Instantaneous(Spawn Group=4, Spawn Time=0.4, Use Spawn Probability=false, UseLoopCountLimit=false) {module=SpawnBurst_Instantaneous004}
  ]
  particleSpawnScript [
   &InitializeParticle(
    Color=<rgba 2 2 2 1>
    Mesh Scale Mode=::NewEnumerator3
    Mesh Uniform Scale=0.25
    Sprite Size Mode=::NewEnumerator3
    Uniform Sprite Size=&SelectFloatFromArray(
     Direct Array Index=&SpawnGroup()
     Float Selection Array=DataInterface() {propertyValues {FloatData [150,150,150,150,150]}}
    )
    UsePositionOffset=false
    Write Material Random=true
   )
   &InitialMeshOrientation(
    Mesh Orientation Mode=::NewEnumerator7
    Rotation=&Multiply_VectorByFloat(
     Vector=&SelectVectorFromArray(
      Direct Array Index=&SpawnGroup()
      Vector Selection Array=DataInterface() {propertyValues {FloatData [<0 0 -0.015>, <0 0 0.03>, <0 0 -0.03>, <0 0 0.015>, <0 0 0>]}}
     )
     Float=&RandomRangeFloat(Minimum=0.5, Maximum=0.75)
    )
    Use Rotation Vector=true
   )
   /SetVariables(Particles.GlowColor=<rgba 0 0.0476 0.2 1>) {isSetParameters=true}
   /SetVariables(Particles.MeshIndex=&SelectIntFromArray(Direct Array Index=&SpawnGroup(), Int Selection Array=DataInterface() {propertyValues {IntData [0,0,0,0,1]}})) {isSetParameters=true}
  ]
  particleUpdateScript [
   &ParticleState()
   &UpdateMeshOrientation(
    Facing Direction=$Emitter.DirectionVector
    Facing Interpolation=false
    Orientation Method=::NewEnumerator1
    Override Quat=false
    TransformFacing=false
    TransformUp=false
   )
   /SetVariables(Particles.Position=&AddVectorToPosition(
     Position=$Engine.Emitter.SimulationPosition
     Vector=&Add_Vector(
      A=&Multiply_VectorByFloat(Vector=$Emitter.DirectionVector, Float=&Lerp_Float(A=25, B=50, Alpha=$Emitter.AlphaFromDistance))
      B=&Multiply_VectorByFloat(
       Vector=&Multiply_VectorByFloat(
        Vector=$Emitter.DirectionVector
        Float=&Multiply_Float(
         A=&SelectFloatFromArray(
          Direct Array Index=&SpawnGroup()
          Float Selection Array=DataInterface() {propertyValues {FloatData [1,1,0.9,0.85,0.8]}}
         )
         B=&Lerp_Float(A=15, B=80, Alpha=$Emitter.AlphaFromDistance)
        )
       )
       Float=&MakeFloatFromInt(Int=&SpawnGroup())
      )
     )
    )) {isSetParameters=true}
   /SetVariables(Particles.GlowPos:NiagaraPosition=&AddVectorToPosition(
     Position=$Particles.Position
     Vector=&Add_Vector(
      A=&RotateVectorByQuaternion(Vector To Rotate=<0 0 6>, Quaternion=$Particles.MeshOrientation)
      B=&RotateVectorByQuaternion(
       Vector To Rotate=&SelectVectorFromArray(
        Direct Array Index=&SpawnGroup()
        Vector Selection Array=DataInterface() {propertyValues {FloatData [<0 0 0>, <0 0 0>, <0 0 0>, <0 0 0>, <-6 0 0>]}}
       )
       Quaternion=$Particles.MeshOrientation
      )
     )
    )) {isSetParameters=true}
   &ScaleMeshSize(Scale Factor=&Multiply_VectorByFloat(
     Vector=&Multiply_VectorByFloat(
      Vector=&VectorFromCurve(VectorCurve=DataInterface() {propertyValues {
         XCurve {keys [
           {interpMode=RCIM_Cubic, tangentMode=RCTM_User, time=0, value=0}
           {arriveTangent=-2.6238, interpMode=RCIM_Cubic, leaveTangent=-2.6238, tangentMode=RCTM_User, time=0.1, value=1.5}
           {interpMode=RCIM_Cubic, tangentMode=RCTM_User, time=0.4, value=1}
           {interpMode=RCIM_Cubic, tangentMode=RCTM_User, time=0.8, value=1}
           {arriveTangent=-8.2706, interpMode=RCIM_Cubic, leaveTangent=-8.2706, tangentMode=RCTM_User, time=1, value=0}
          ]}
         YCurve {keys [
           {interpMode=RCIM_Cubic, tangentMode=RCTM_User, time=0, value=0}
           {arriveTangent=-3.5474, interpMode=RCIM_Cubic, leaveTangent=-3.5474, tangentMode=RCTM_User, time=0.1, value=1.8}
           {interpMode=RCIM_Cubic, tangentMode=RCTM_User, time=0.4, value=1}
           {interpMode=RCIM_Cubic, tangentMode=RCTM_User, time=0.8, value=1}
           {arriveTangent=-8.2706, interpMode=RCIM_Cubic, leaveTangent=-8.2706, tangentMode=RCTM_User, time=1, value=0}
          ]}
         ZCurve {keys [
           {interpMode=RCIM_Cubic, tangentMode=RCTM_User, time=0, value=0}
           {arriveTangent=-4.2603, interpMode=RCIM_Cubic, leaveTangent=-4.2603, tangentMode=RCTM_User, time=0.1, value=2}
           {interpMode=RCIM_Cubic, tangentMode=RCTM_User, time=0.4, value=1}
           {interpMode=RCIM_Cubic, tangentMode=RCTM_User, time=0.8, value=1}
           {arriveTangent=-8.0363, interpMode=RCIM_Cubic, leaveTangent=-8.0363, tangentMode=RCTM_User, time=1, value=0}
          ]}
        }})
      Float=&SelectFloatFromArray(
       Direct Array Index=&SpawnGroup()
       Float Selection Array=DataInterface() {propertyValues {FloatData [1,0.85,0.7,0.55,0.7]}}
      )
     )
     Float=&Lerp_Float(A=0.4, B=1, Alpha=$Emitter.AlphaFromDistance)
    ))
   &ScaleSpriteSize(
    Scale Sprite Size Mode=::NewEnumerator0
    Uniform Scale Factor=&MakeFloatFromVector(Vector=$Particles.Scale, Channel=::NewEnumerator0)
   )
   &DynamicMaterialParameters(
    Index 0 Param 2=1
    Index 0 Param 4=1
    Module.Index0Param0Write=true
    Module.Index0Param1Write=true
    Module.Index0Param2Write=true
    Module.Index0Param3Write=true
   )
   /SetVariables(Particles.DynMatGlow:Vector4f=&MakeVector4(
     X=&MakeFloatFromVector(
      Vector=&RotateVectorByQuaternion(Vector To Rotate=<1 0 0>, Quaternion=$Particles.MeshOrientation)
      Channel=::NewEnumerator0
     )
     Y=&MakeFloatFromVector(
      Vector=&RotateVectorByQuaternion(Vector To Rotate=<1 0 0>, Quaternion=$Particles.MeshOrientation)
      Channel=::NewEnumerator1
     )
     Z=&MakeFloatFromVector(
      Vector=&RotateVectorByQuaternion(Vector To Rotate=<1 0 0>, Quaternion=$Particles.MeshOrientation)
      Channel=::NewEnumerator2
     )
     W=&Multiply_Float(A=300, B=&MakeFloatFromVector(Vector=$Particles.Scale, Channel=::NewEnumerator1))
    )) {isSetParameters=true}
  ]
  rendererClasses [&NiagaraMeshRendererProperties, &NiagaraSpriteRendererProperties]
  renderers [
   {rendererClass=&NiagaraMeshRendererProperties, rendererIndex=0}
   {rendererClass=&NiagaraSpriteRendererProperties, rendererIndex=1}
  ]
 }}
summary {
 systemName=NS_Athena_BootyDirection
 emitters [{
   emitterName=NE_Direction_Indicators
   rendererClasses [&NiagaraMeshRendererProperties, &NiagaraSpriteRendererProperties]
  }]
 userVariables [User.BootyActorPos {type=NiagaraPosition, description="", defaultValue {}}]
}
