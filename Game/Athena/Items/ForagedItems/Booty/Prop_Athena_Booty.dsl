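; Buried treasure dig-spot prop that players hit with harvesting tools to unearth a Booty chest.
; The server traces the ground normal and spawns the chest, then each hit raises and tilts it, plays Niagara FX and grows a dirt mound.
; RepNotify MaxDug ends digging, disabling the hitbox and flagging the chest dug up.
; Gameplay events tell the owner to toggle markers on overlap and force-drop on opening, flagging whether the owner opened it.
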(event Custom|ChestOpened
  (params (SelfActor :type Building%20Actor%20Object%20Reference) (InteractingPawn :type Fort%20Pawn%20Object%20Reference))
  ;[1 1 1] Force Drop the Item if the owner opens the chest
  (Utilities|FlowControl|Delay 0.1)
  (Ability|SendGameplayEventtoActor (Actor|GetOwner) "(TagName=\"Event.Athena.Booty.ForceDrop\")" (Utilities|Struct|MakeGameplayEventData :ContextHandle "()" :InstigatorTags "(GameplayTags=)" :TargetTags "(GameplayTags=)" :EventMagnitude (Math|Conversions|ToFloat(Boolean) (Utilities|Equal(Object) InteractingPawn (Actor|GetOwner))) :TargetData "()")))

(event Custom|OnBeenDug
  (bind _chestobject (Variables|Default|GetChestObject))
  ;[1 1 1] Sound indicator
  (Variables|Default|SetBeenDug true)
  (if (not (Variables|/Game/Athena/Items/ForagedItems/Booty/B_Athena_Booty_Chest.B_Athena_Booty_Chest_C|GetShowSound _chestobject))
    (Variables|/Game/Athena/Items/ForagedItems/Booty/B_Athena_Booty_Chest.B_Athena_Booty_Chest_C|SetShowSound true _chestobject)))

(event EventBeginPlay
  ;[1 1 1] Spawn In the Chest
  (if (Networking|IsServer)
    (CallFunction|FindPlacementNormal)
    (bind _returnvalue (Game|SpawnActorfromClass "/Game/Athena/Items/ForagedItems/Booty/B_Athena_Booty_Chest.B_Athena_Booty_Chest_C" (Math|Transform|MakeTransform (+ (Transformation|GetActorLocation) (Variables|Default|GetZOffset)) (Transformation|GetActorRotation)) "AlwaysSpawn" "MultiplyWithRoot" (Variables|Actor|GetInstigator) 0 0 (Variables|Actor|GetInstigator)))
    (bind _output_get (Variables|Default|SetChestObject _returnvalue))
    (Variables|Default|SetChestInitialLoc (Transformation|GetActorLocation _output_get))
    ;[1 1 1] Bind to chest opened
    (Interaction|BindEventtoOnInteract _output_get (delegate Custom|ChestOpened))
    ;[1 1 1] Bind Damage and Call Replecated Event
    (EventDispatchers|AssignOnDamaged (delegate Custom|OnDamageTaken))
    (else
      (Development|PrintString "Client Booty Prop: I AM HERE!"))))

(event Custom|OnDamageTaken
  (params (DamagedActor :type Actor%20Object%20Reference) (Damage :type Float%20%28single-precision%29) (InstigatedBy :type Controller%20Object%20Reference) (DamageCauser :type Actor%20Object%20Reference) (HitLocation :type Vector) (FHitComponent :type Primitive%20Component%20Object%20Reference) (BoneName :type Name) (Momentum :type Vector))
  ;[1 1 1] Bind Damage and Call Replecated Event
  (bind _asfort_weapon (Utilities|Casting|CastToFortWeapon DamageCauser))
  (if (Utilities|Enum|Equal(Enum) (Item|GetWeaponDataItemType _asfort_weapon) "WeaponHarvest")
    (Math|Integer|IncrementInt (Variables|Default|GetNumTimesHit)
      (:%20%20
        (CallFunction|TriggerDigFX HitLocation)
        (CallFunction|MoveChest)
        (CallFunction|UnhideDirt)))))

(event Custom|OnMaxDug
  ;[1 1 1] After Max Hits Unbind damage events and set collisions
  (Development|PrintString "MaxDug")
  (EventDispatchers|UnbindallEventsfromOnDamaged)
  (Collision|SetCollisionEnabled (Variables|Default|GetHitbox) "NoCollision")
  (if (Networking|IsServer)
    (Variables|/Game/Athena/Items/ForagedItems/Booty/B_Athena_Booty_Chest.B_Athena_Booty_Chest_C|SetDugUp true (Variables|Default|GetChestObject))))

(event Custom|TriggerDigFX
  (params (Location :type Vector))
  ;[1 1 1] Spawn FX at Location
  (Niagara|SpawnSystematLocation (Variables|Default|GetNS_DigParticle) (+ (Transformation|GetWorldLocation (Variables|Default|GetDirt1)) "0, 0,55.000000") (Transformation|GetWorldRotation (Variables|Default|GetDirt1)))
  (if (>= (Variables|Default|GetNumTimesHit) (Variables|Default|GetMaxTimesHit))
    (Niagara|SpawnSystematLocation (Variables|Default|GetNS_DigParticleFinal) (Transformation|GetWorldLocation (Variables|Default|GetDirt1)) (Transformation|GetWorldRotation (Variables|Default|GetDirt1)))))

(event Custom|ClientMoveChestLoc
  (params (ChestLoc :type Vector))
  (bind _chestobject (Variables|Default|GetChestObject))
  ;[1 1 1] Client replicated movements
  (Utilities|IsValid _chestobject
    (:Is%20Valid
      (Transformation|SetActorLocation _chestobject ChestLoc false false))))

(event Custom|MoveChest
  (bind (_returnvalue_roll _returnvalue_pitch _returnvalue_yaw) (Math|Rotator|BreakRotator (Math|Rotator|MakeRotfromZX (Variables|Default|GetChestRotNormal) (Transformation|GetActorLocation (Variables|Default|GetChestObject)))))
  (bind (_returnvalue_roll_2 _returnvalue_pitch_2 _returnvalue_yaw_2) (Math|Rotator|BreakRotator (Math|Rotator|MakeRotfromZX (Variables|Default|GetChestRotNormal) (Transformation|GetActorLocation))))
  (bind (_output_roll _output_pitch) (Math|Rotator|BreakRotator (Utilities|Array|Get(aref) (Variables|Default|GetChestRotPattern) (- (Variables|Default|GetNumTimesHit) 1))))
  (bind _numtimeshit (Variables|Default|GetNumTimesHit))
  ;[1 1 1] Handle Chest Damage
  (Development|PrintString "Dig" true true "(R=1.000000,G=0.000000,B=0.964451,A=1.000000)")
  ;[1 1 1] Set Times Dug on the chest
  (Variables|/Game/Athena/Items/ForagedItems/Booty/B_Athena_Booty_Chest.B_Athena_Booty_Chest_C|SetTimesDug (Variables|Default|GetNumTimesHit) (Variables|Default|GetChestObject))
  ;[1 1 1] Move up and rotate on hit
  (bind _output_get (Variables|/Game/Athena/Items/ForagedItems/Booty/B_Athena_Booty_Chest.B_Athena_Booty_Chest_C|SetChestCurrentLoc (+ (Variables|Default|GetChestInitialLoc) (* (Variables|Default|GetChestRaisePerHit) _numtimeshit)) (Variables|Default|GetChestObject)))
  (CallFunction|ClientMoveChestLoc _output_get)
  (if (>= (Variables|Default|GetNumTimesHit) (Variables|Default|GetMaxTimesHit))
    ;;[1 1 1] If it is the last hit, set max dug and set the alternate rotation
    (bind _output_get_2 (Variables|/Game/Athena/Items/ForagedItems/Booty/B_Athena_Booty_Chest.B_Athena_Booty_Chest_C|SetChestCurrentRot (Math|Rotator|MakeRotator _returnvalue_roll _returnvalue_pitch _returnvalue_yaw) (Variables|Default|GetChestObject)))
    (CallFunction|ClientMoveChestRot _output_get_2)
    (Variables|Default|SetMaxDug true)
    (else
      (bind _output_get_3 (Variables|/Game/Athena/Items/ForagedItems/Booty/B_Athena_Booty_Chest.B_Athena_Booty_Chest_C|SetChestCurrentRot (Math|Rotator|MakeRotator (+ _output_roll _returnvalue_roll_2) (+ _output_pitch _returnvalue_pitch_2) _returnvalue_yaw_2) (Variables|Default|GetChestObject)))
      (CallFunction|ClientMoveChestRot _output_get_3)
      (if (not (Variables|Default|GetBeenDug))
        (CallFunction|OnBeenDug)))))

(event Custom|ClientMoveChestRot
  (params (ChestRot :type Rotator))
  (bind _chestobject (Variables|Default|GetChestObject))
  ;[1 1 1] Client replicated movements
  (Utilities|IsValid _chestobject
    (:Is%20Valid
      (Transformation|SetActorRotation _chestobject ChestRot false))))

(event Custom|CheckOnOwnerSwitch
  (params (NewOwner :type Actor%20Object%20Reference))
  ;[1 1 1] Check On OwnerSwitch
  (if (Collision|IsOverlappingActor (Variables|Default|GetOwnerOverlap) NewOwner :!class /Script/Engine.PrimitiveComponent)
    ;[1 1 1] When Over Overlaps Turn send events to toggle ground markers
    (Ability|SendGameplayEventtoActor (Actor|GetOwner) "(TagName=\"Event.Athena.Booty.BeginOverlap\")")))

(event Custom|UnhideDirt
  (bind (_returnvalue_roll _returnvalue_pitch _returnvalue_yaw) (Math|Rotator|BreakRotator (Math|Rotator|MakeRotfromZX (Variables|Default|GetChestRotNormal) (Transformation|GetActorLocation))))
  ;[1 1 1] Unhide dirt
  (if (not (>= (Variables|Default|GetNumTimesHit) 5))
    (Variables|Default|SetDirtTransform (Math|Transform|MakeTransform (+ (Variables|/Script/Engine.SceneComponent|GetRelativeLocation (Variables|Default|GetDirt1)) (Variables|Default|GetDirtMoveOnHit)) (Math|Rotator|MakeRotator _returnvalue_roll _returnvalue_pitch _returnvalue_yaw) (+ (Variables|/Script/Engine.SceneComponent|GetRelativeScale3D (Variables|Default|GetDirt1)) (Variables|Default|GetDirtScaleOnHit))))))

(event OnComponentBeginOverlap(OwnerOverlap)
  (params (OverlappedComponent :type Primitive%20Component%20Object%20Reference) (OtherActor :type Actor%20Object%20Reference) (OtherComp :type Primitive%20Component%20Object%20Reference) (OtherBodyIndex :type Integer) (bFromSweep :type Boolean) (SweepResult :type Hit%20Result%20Structure%20%28by%20ref%29))
  (bind _returnvalue (Actor|GetOwner))
  ;[1 1 1] When Over Overlaps Turn send events to toggle ground markers
  (if (Utilities|Equal(Object) OtherActor _returnvalue)
    (Ability|SendGameplayEventtoActor _returnvalue "(TagName=\"Event.Athena.Booty.BeginOverlap\")")))

(event OnComponentEndOverlap(OwnerOverlap)
  (params (OverlappedComponent :type Primitive%20Component%20Object%20Reference) (OtherActor :type Actor%20Object%20Reference) (OtherComp :type Primitive%20Component%20Object%20Reference) (OtherBodyIndex :type Integer))
  (bind _returnvalue (Actor|GetOwner))
  ;[1 1 1] When Over Overlaps Turn send events to toggle ground markers
  (if (Utilities|Equal(Object) OtherActor _returnvalue)
    (if (Collision|GetActorEnableCollision)
      (Ability|SendGameplayEventtoActor _returnvalue "(TagName=\"Event.Athena.Booty.EndOverlap\")"))))

(fn FindPlacementNormal
  (bind _outhit (Collision|LineTraceByChannel (+ (Transformation|GetActorLocation) (Variables|Default|GetTraceForChestOffset)) (- (Transformation|GetActorLocation) (Variables|Default|GetTraceForChestOffset)) "TraceTypeQuery10"))
  (Variables|Default|SetChestRotNormal (.Normal (Collision|BreakHitResult _outhit))))

(fn OnRep_ChestCurrentPos
  (Variables|/Game/Athena/Items/ForagedItems/Booty/B_Athena_Booty_Chest.B_Athena_Booty_Chest_C|SetChestCurrentLoc (Variables|Default|GetChestCurrentPos) (Variables|Default|GetChestObject)))

(fn OnRep_ChestCurrentRot
  (Variables|/Game/Athena/Items/ForagedItems/Booty/B_Athena_Booty_Chest.B_Athena_Booty_Chest_C|SetChestCurrentRot (Variables|Default|GetChestCurrentRot) (Variables|Default|GetChestObject)))

(fn OnRep_DirtTransform
  (bind _dirttransform (Variables|Default|GetDirtTransform))
  (bind _dirt1 (Variables|Default|GetDirt1))
  (Transformation|SetWorldRotation _dirt1 (.rotation _dirttransform) false false) ; This one being world not relative is intentional
  (Rendering|SetVisibility _dirt1 true false)
  (Transformation|SetRelativeLocation _dirt1 (.location _dirttransform) false false)
  (Transformation|SetRelativeScale3D _dirt1 (.scale _dirttransform)))

(fn OnRep_MaxDug
  (if (Variables|Default|GetMaxDug)
    (CallFunction|OnMaxDug)))

(fn ShouldDie
  (params (Damage :type Float%20%28single-precision%29) (EventInstigator :type Controller%20Object%20Reference) (DamageCauser :type Actor%20Object%20Reference) (bIsTest :type Boolean) (EffectContext :type Gameplay%20Effect%20Context%20Handle%20Structure%20%28by%20ref%29))
  (return false))

(fn ConstructionScript
  (Collision|SetCollisionEnabled (Variables|Debug|GetStaticMeshComponent) "NoCollision"))
